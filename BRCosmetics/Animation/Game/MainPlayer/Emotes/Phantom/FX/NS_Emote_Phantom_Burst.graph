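# Niagara one-shot burst effect for the Phantom emote: a short, colorful spark pop with a glow flash.
# Particle scatters dragged sprites colored through nested random bools, re-rolling a secondary glow color every update so it flickers.
# Glow adds a camera-offset, shrinking red-tinted flash and writes light attributes, though only a sprite renderer is listed.
# Time-of-day scalars tune brightness; a front-end preview flag reduces the flash and zeroes light radius.
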NiagaraSystem /BRCosmetics/Animation/Game/MainPlayer/Emotes/Phantom/FX/NS_Emote_Phantom_Burst
dependencies {
 usedDataInterfaces [&NiagaraDataInterfaceCurve, &NiagaraDataInterfaceColorCurve]
 usedDynamicInputs []
 usedModules [&SystemState]
}
emitters {
 Particle {
  emitterUpdateScript [
   &EmitterState(Life Cycle Mode=::NewEnumerator1, Loop Behavior=::NewEnumerator1, Loop Duration=1, UseLoopDelay=false)
   &SpawnBurst_Instantaneous(Spawn Count=20, Use Spawn Probability=false, UseLoopCountLimit=false)
  ]
  particleSpawnScript [
   &InitializeParticle(
    Color=&MakeCustomLinearColorFromBool(
     Bool=&RandomBool()
     True Color=&MakeCustomLinearColorFromBool(
      Bool=&RandomBool()
      True Color=&MakeCustomLinearColorFromBool(Bool=&RandomBool(), True Color=<rgba 1 0.07 0 1>, False Color=<rgba 1 0.8281 0 1>)
      False Color=&MakeCustomLinearColorFromBool(Bool=&RandomBool(), True Color=<rgba 0 1 0.1767 1>, False Color=<rgba 0 1 0.8022 1>)
     )
     False Color=&MakeCustomLinearColorFromBool(
      Bool=&RandomBool()
      True Color=&MakeCustomLinearColorFromBool(Bool=&RandomBool(), True Color=<rgba 0 0.1339 1 1>, False Color=<rgba 0 1 0.8617 1>)
      False Color=&MakeCustomLinearColorFromBool(Bool=&RandomBool(), True Color=<rgba 0.5824 0 1 1>, False Color=<rgba 1 0.8281 0 1>)
     )
    )
    Lifetime=&RandomRangeFloat(Minimum=0.15, Maximum=0.4)
    Sprite Rotation Angle=45
    Sprite Rotation Mode=::NewEnumerator0
    Sprite Size Mode=::NewEnumerator1
    Uniform Sprite Size Max=2
    Uniform Sprite Size Min=1
    UsePositionOffset=false
    Write Material Random=true
   )
   /SetVariables(Particles.RandomCore:NiagaraFloat=&FloatFromCurve(
     FloatCurve=DataInterface() {propertyValues {Curve {keys [
         {arriveTangent=0.0559, interpMode=RCIM_Cubic, leaveTangent=0.0559, tangentMode=RCTM_User, time=0, value=0.2}
         {arriveTangent=0.1895, interpMode=RCIM_Cubic, leaveTangent=0.1895, tangentMode=RCTM_User, time=0.9, value=0.25}
         {arriveTangent=1.1808, interpMode=RCIM_Cubic, leaveTangent=1.1808, tangentMode=RCTM_User, time=1, value=0.3}
        ]}}}
     CurveIndex=&RandomRangeFloat(Minimum=0, Maximum=1)
     Scale Curve=2
    )) {isSetParameters=true}
   &ShapeLocation(MaskSpawnGroup=false, Offset Mode=::NewEnumerator0, Override Seed=false, Sphere Radius=5, UseOwnerScale=true)
   &AddVelocity(Module.Use Falloff Distance=false, Override Seed=false, Velocity Mode=::NewEnumerator1, Velocity Speed=150)
  ]
  particleUpdateScript [
   &ParticleState()
   /SetVariables(Particles.Glow03_Color=&MakeLinearColorFromVectorAndFloat(
     "Vector (RGB)"=&MakeVectorFromLinearColorRGB(Color=&MakeCustomLinearColorFromBool(
       Bool=&RandomBool()
       True Color=&MakeCustomLinearColorFromBool(
        Bool=&RandomBool()
        True Color=&MakeCustomLinearColorFromBool(Bool=&RandomBool(), True Color=<rgba 1 0.07 0 1>, False Color=<rgba 1 0.8281 0 1>)
        False Color=&MakeCustomLinearColorFromBool(Bool=&RandomBool(), True Color=<rgba 0 1 0.1767 1>, False Color=<rgba 0 1 0.8022 1>)
       )
       False Color=&MakeCustomLinearColorFromBool(
        Bool=&RandomBool()
        True Color=&MakeCustomLinearColorFromBool(Bool=&RandomBool(), True Color=<rgba 0 0.1339 1 1>, False Color=<rgba 0 1 0.8617 1>)
        False Color=&MakeCustomLinearColorFromBool(Bool=&RandomBool(), True Color=<rgba 0.5824 0 1 1>, False Color=<rgba 1 0.8281 0 1>)
       )
      ))
     "Float (Alpha)"=&Multiply_Float(
      A=&Multiply_Float(
       A=&FloatFromCurve(FloatCurve=DataInterface() {propertyValues {Curve {keys [<at 0 0 RCIM_Cubic>,<at 0.5 1 RCIM_Cubic>,<at 1 0 RCIM_Cubic>]}}})
       B=&FloatFromCurve(FloatCurve=DataInterface() {propertyValues {Curve {keys [<at 0 1>,<at 1 0>]}}}, CurveIndex=$Emitter.Age)
      )
      B=&FNTimeOfDayScalar(Night Scale=0.2, Morning Scale=0.3, Day Scale=0.5, Evening Scale=0.3)
     )
    )) {isSetParameters=true}
   /SetVariables(Particles.Glow03_Size=&Multiply_Vector2D_ByFloat(
     Float=&FloatFromCurve(
      FloatCurve=DataInterface() {propertyValues {Curve {keys [
          <at 0 0 RCIM_Cubic>
          <at 0.2 1.3 RCIM_Cubic>
          {arriveTangent=-0.1423, interpMode=RCIM_Cubic, leaveTangent=-0.1423, tangentMode=RCTM_User, time=1, value=0}
         ]}}}
      Scale Curve=10
     )
     Vector2D=<1 1>
    )) {isSetParameters=true}
   &CurlNoiseForce(Noise Strength=50, Pan Noise=true, Pan Noise Field=<0.5 0.5 0>) {enabled=false}
   &ScaleSpriteSize(Scale Factor=&Vector2DFromFloat(Value=&FloatFromCurve(
      FloatCurve=DataInterface() {propertyValues {Curve {keys [
          <at 0 1 RCIM_Cubic>
          {arriveTangent=-1, interpMode=RCIM_Cubic, leaveTangent=-1, time=0.6, value=1.25}
          {arriveTangent=-0.1423, interpMode=RCIM_Cubic, leaveTangent=-0.1423, tangentMode=RCTM_User, time=1, value=0}
         ]}}}
      Scale Curve=1.5
     )))
   &ScaleColor(
    Scale Alpha=&Multiply_Float(
     A=&Multiply_Float(
      A=1
      B=&FloatFromCurve(FloatCurve=DataInterface() {propertyValues {Curve {keys [<at 0 1>,<at 1 0>]}}}, CurveIndex=$Emitter.Age)
     )
     B=&FNTimeOfDayScalar(Night Scale=0.5, Morning Scale=0.4, Evening Scale=0.3)
    )
    Scale RGB=<5 5 5>
    ScaleA=true
    ScaleRGB=true
   )
   &Drag(Drag=5, Use Linear Drag=true, Use Rotational Drag=false)
   &SolveForcesAndVelocity(Clamp Velocity=false, Limit Acceleration=false)
  ]
  rendererClasses [&NiagaraSpriteRendererProperties]
  renderers [
   {rendererClass=&NiagaraSpriteRendererProperties}
   {rendererClass=&NiagaraSpriteRendererProperties, rendererIndex=1}
  ]
 }
 glow {
  emitterUpdateScript [
   &EmitterState(Inactive Response=::NewEnumerator1, Life Cycle Mode=::NewEnumerator1, Loop Behavior=::NewEnumerator1, Loop Duration=1, UseLoopDelay=false)
   &SpawnBurst_Instantaneous(Use Spawn Probability=false, UseLoopCountLimit=false)
  ]
  particleSpawnScript [
   &InitializeParticle(
    Color=<rgba 0.5 0.5 0.5 1>
    Lifetime=0.35
    Sprite Size Mode=::NewEnumerator3
    Uniform Sprite Size=150
    UsePositionOffset=true
    Write Material Random=true
   )
   &CameraOffset(Camera Offset Amount=5)
  ]
  particleUpdateScript [
   &ParticleState()
   &ScaleSpriteSize(
    Uniform Curve Scale=&MakeCustomFloatFromBool(Bool=$User.bIsFrontEndPreview, True Float=0.5, False Float=1)
    Uniform Curve Sprite Scale=DataInterface() {propertyValues {Curve {keys [<at 0 1>, {tangentMode=RCTM_User, time=1, value=0.2}]}}}
   )
   /SetVariables(Particles.Size002:Vector2f=&Multiply_Vector2D(A=$Particles.SpriteSize, B=&Vector2DFromFloat(Value=0.5))) {isSetParameters=true}
   &ScaleColor(
    Scale Alpha=&FloatFromCurve(
     FloatCurve=DataInterface() {propertyValues {Curve {keys [
         {arriveTangent=8.4201, interpMode=RCIM_Cubic, leaveTangent=8.4201, tangentMode=RCTM_User, time=0, value=-0.025}
         {arriveTangent=0.2873, interpMode=RCIM_Cubic, leaveTangent=-0.1265, tangentMode=RCTM_Break, time=0.2, value=0.8}
         {arriveTangent=-2.3949, interpMode=RCIM_Cubic, leaveTangent=-2.3949, tangentMode=RCTM_User, time=0.5, value=0.4}
         <at 1 0 RCIM_Cubic>
        ]}}}
     Scale Curve=&Multiply_Float(
      A=&FNTimeOfDayScalar(Night Scale=0.7, Morning Scale=1, Evening Scale=1)
      B=&MakeCustomFloatFromBool(Bool=$User.bIsFrontEndPreview, True Float=0.5, False Float=1)
     )
    )
    Scale RGB=&MakeVectorFromLinearColorRGB(Color=&ColorFromCurve(ColorCurve=DataInterface() {propertyValues {
        AlphaCurve {keys []}
        BlueCurve {keys [<at 0.7006 0.06 RCIM_Cubic>]}
        GreenCurve {keys [<at 0.7006 0.025 RCIM_Cubic>]}
        RedCurve {keys [<at 0.7006 0.3 RCIM_Cubic>]}
       }}))
    ScaleA=true
    ScaleRGB=true
   )
   /SetVariables(Particles.color002:LinearColor=&MakeLinearColorFromVectorAndFloat(
     "Vector (RGB)"=&MakeVectorFromLinearColorRGB(Color=<rgba 1 1 1 1>)
     "Float (Alpha)"=&FloatFromCurve(
      FloatCurve=DataInterface() {propertyValues {Curve {keys [
          {arriveTangent=-1.1243, interpMode=RCIM_Cubic, leaveTangent=-1.1243, tangentMode=RCTM_User, time=0, value=1}
          {arriveTangent=-2.3022, interpMode=RCIM_Cubic, leaveTangent=-2.3022, tangentMode=RCTM_User, time=0.2371, value=0.653}
          {arriveTangent=-0.0759, interpMode=RCIM_Cubic, leaveTangent=-0.0759, tangentMode=RCTM_User, time=0.6, value=0}
         ]}}}
      Scale Curve=&Multiply_Float(
       A=&FNTimeOfDayScalar(Night Scale=0.6, Morning Scale=0.7, Evening Scale=0.7)
       B=&FN_FrontEndFloatModifier(Gameplay=0.6, Front End=0.4)
      )
     )
    )) {isSetParameters=true}
   &FN_Light_Attributes(
    Color=&ColorFromCurve(ColorCurve=DataInterface() {propertyValues {AlphaCurve {keys []}, BlueCurve {keys [<at 0.1896 0.028>]}, GreenCurve {keys [<at 0.1896 0.021>]}, RedCurve {keys [<at 0.1896 1.5>]}}})
    Create Color Attribute=true
    Create Exponent Attribute=true
    Create Position Attribute=true
    Create Radius Attribute=true
    Exponent=2
    Radius=&FloatFromCurve(
     FloatCurve=DataInterface() {propertyValues {Curve {keys [<at 0 0.4>,<at 0.2 1.1>,<at 1 1.4>]}}}
     Scale Curve=&MakeCustomFloatFromBool(Bool=$User.bIsFrontEndPreview, True Float=0, False Float=100)
    )
    Scale Color=&FloatFromCurve(
     FloatCurve=DataInterface() {propertyValues {Curve {keys [
         {arriveTangent=4.1988, interpMode=RCIM_Cubic, leaveTangent=4.1988, tangentMode=RCTM_User, time=0, value=0}
         {arriveTangent=-0.0105, interpMode=RCIM_Cubic, leaveTangent=-0.0105, tangentMode=RCTM_User, time=0.15, value=1}
         {arriveTangent=-0.9284, interpMode=RCIM_Cubic, leaveTangent=-0.9284, tangentMode=RCTM_User, time=0.5, value=0}
        ]}}}
     Scale Curve=&Multiply_Float(A=&FNTimeOfDayScalar(Night Scale=0.5, Morning Scale=1, Evening Scale=1), B=20)
    )
   )
   &Drag(Drag=2, Use Linear Drag=true, Use Rotational Drag=true)
   &SolveForcesAndVelocity(Clamp Velocity=false, Limit Acceleration=false)
  ]
  rendererClasses [&NiagaraSpriteRendererProperties]
  renderers [{rendererClass=&NiagaraSpriteRendererProperties}]
 }
}
summary {
 systemName=NS_Emote_Phantom_Burst
 emitters [
  {emitterName=glow, rendererClasses [&NiagaraSpriteRendererProperties]}
  {emitterName=Particle, rendererClasses [&NiagaraSpriteRendererProperties]}
 ]
 userVariables [
  User.bIsFrontEnd {type=NiagaraBool}
  User.bIsFrontEndPreview {type=NiagaraBool}
 ]
}
