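# Niagara hand-trail effect for the Phantom emote, built on two skeletal mesh bone/socket indices.
# Ribbon_L and Ribbon_R each trail one index with narrowing, fading ribbons, while Particle spawns small random rainbow-colored glow sprites at both, drifting with curl noise.
# Smoke re-samples the skeleton every update to stay attached as tinted rotating puffs.
# Alpha scales by time of day, by platform for ribbons and smoke, and smoke dims in front-end preview.
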NiagaraSystem /BRCosmetics/Animation/Game/MainPlayer/Emotes/Phantom/FX/NS_Emote_Phantom_HandTrail
dependencies {
 usedDataInterfaces [
  &NiagaraDataInterfaceSpriteRendererInfo
  &NiagaraDataInterfaceArrayInt32
  &NiagaraDataInterfaceCurve
  &NiagaraDataInterfacePlatformSet
  &NiagaraDataInterfaceColorCurve
  &NiagaraDataInterfaceSkeletalMesh
 ]
 usedDynamicInputs []
 usedModules [
  &self:SystemSpawnScript.SystemScriptSource.SystemScriptGraph.NiagaraNodeAssignment_0.SetVariables_#1
  &SystemState
 ]
}
emitters {
 Particle {
  emitterUpdateScript [
   &EmitterState(Life Cycle Mode=::NewEnumerator1, Loop Duration=1, UseLoopDelay=false)
   &SpawnRate(SpawnRate=60, Use Spawn Probability=false)
  ]
  particleSpawnScript [
   &PartitionParticles(Number Of Partitions=3)
   &InitializeParticle(
    Color=&MakeCustomLinearColorFromBool(
     Bool=&RandomBool()
     True Color=&MakeCustomLinearColorFromBool(
      Bool=&RandomBool()
      True Color=&MakeCustomLinearColorFromBool(Bool=&RandomBool(), True Color=<rgba 1 0.07 0 1>, False Color=<rgba 1 0.8281 0 1>)
      False Color=&MakeCustomLinearColorFromBool(Bool=&RandomBool(), True Color=<rgba 0 1 0.1767 1>, False Color=<rgba 0 1 0.8022 1>)
     )
     False Color=&MakeCustomLinearColorFromBool(
      Bool=&RandomBool()
      True Color=&MakeCustomLinearColorFromBool(Bool=&RandomBool(), True Color=<rgba 0 0.1339 1 1>, False Color=<rgba 0 1 0.8617 1>)
      False Color=&MakeCustomLinearColorFromBool(Bool=&RandomBool(), True Color=<rgba 0.5824 0 1 1>, False Color=<rgba 1 0.8281 0 1>)
     )
    )
    Lifetime=&RandomRangeFloat(Minimum=0.2, Maximum=0.25)
    Sprite Rotation Mode=::NewEnumerator2
    Sprite Size Mode=::NewEnumerator1
    Uniform Sprite Size Max=2
    Uniform Sprite Size Min=1
    UsePositionOffset=false
    Write Material Random=true
   )
   &SkeletalMeshLocation(
    Bone / Socket Index=&SelectIntFromArray(Direct Array Index=$Particles.PartitionParticles.Partition, Int Selection Array=DataInterface() {propertyValues {IntData [1,2]}})
    MaskSpawnGroup=false
    Mesh Sampling Type=::NewEnumerator5
    Sampled Position Offset=&AddVectorToPosition(Position=<0 0 0>, Vector=$Particles.ShapeLocation.ShapeVector)
    Skeletal Mesh=$System.Skeletal Mesh
    Socket Sampling Mode=::NewEnumerator5
   )
   /SetVariables(Particles.RandomCore:NiagaraFloat=&FloatFromCurve(
     FloatCurve=DataInterface() {propertyValues {Curve {keys [
         {arriveTangent=0.0559, interpMode=RCIM_Cubic, leaveTangent=0.0559, tangentMode=RCTM_User, time=0, value=0.2}
         {arriveTangent=0.1895, interpMode=RCIM_Cubic, leaveTangent=0.1895, tangentMode=RCTM_User, time=0.9, value=0.25}
         {arriveTangent=1.1808, interpMode=RCIM_Cubic, leaveTangent=1.1808, tangentMode=RCTM_User, time=1, value=0.3}
        ]}}}
     CurveIndex=&RandomRangeFloat(Minimum=0, Maximum=1)
     Scale Curve=2
    )) {isSetParameters=true}
   &ShapeLocation(MaskSpawnGroup=false, Non Uniform Scale=<0.5 0.5 1>, Offset Mode=::NewEnumerator0, Override Seed=false, Sphere Radius=10, UseOwnerScale=true)
   &InheritVelocity(Inherited Velocity Speed Limit=50, Speed Limit=true, UseSpeedThreshold=false) {enabled=false}
  ]
  particleUpdateScript [
   &ParticleState()
   /SetVariables(Particles.Glow03_Color=&MakeLinearColorFromVectorAndFloat(
     "Vector (RGB)"=&MakeVectorFromLinearColorRGB(Color=&MakeCustomLinearColorFromBool(
       Bool=&RandomBool()
       True Color=&MakeCustomLinearColorFromBool(
        Bool=&RandomBool()
        True Color=&MakeCustomLinearColorFromBool(Bool=&RandomBool(), True Color=<rgba 1 0.07 0 1>, False Color=<rgba 1 0.8281 0 1>)
        False Color=&MakeCustomLinearColorFromBool(Bool=&RandomBool(), True Color=<rgba 0 1 0.1767 1>, False Color=<rgba 0 1 0.8022 1>)
       )
       False Color=&MakeCustomLinearColorFromBool(
        Bool=&RandomBool()
        True Color=&MakeCustomLinearColorFromBool(Bool=&RandomBool(), True Color=<rgba 0 0.1339 1 1>, False Color=<rgba 0 1 0.8617 1>)
        False Color=&MakeCustomLinearColorFromBool(Bool=&RandomBool(), True Color=<rgba 0.5824 0 1 1>, False Color=<rgba 1 0.8281 0 1>)
       )
      ))
     "Float (Alpha)"=&Multiply_Float(
      A=&FloatFromCurve(
       FloatCurve=DataInterface() {propertyValues {Curve {keys [
           {interpMode=RCIM_Cubic, tangentMode=RCTM_User, time=0, value=1}
           {arriveTangent=-0.815, interpMode=RCIM_Cubic, leaveTangent=-0.815, tangentMode=RCTM_User, time=1, value=0}
          ]}}}
       Scale Curve=&FNTimeOfDayScalar(Night Scale=0.5, Morning Scale=0.3, Evening Scale=0.3)
      )
      B=0.23
     )
    )) {isSetParameters=true}
   /SetVariables(Particles.Glow03_Size=&Multiply_Vector2D_ByFloat(
     Float=&FloatFromCurve(
      FloatCurve=DataInterface() {propertyValues {Curve {keys [
          <at 0 0 RCIM_Cubic>
          <at 0.2 1.3 RCIM_Cubic>
          {arriveTangent=-0.1423, interpMode=RCIM_Cubic, leaveTangent=-0.1423, tangentMode=RCTM_User, time=1, value=0}
         ]}}}
      Scale Curve=10
     )
     Vector2D=<1 1>
    )) {isSetParameters=true}
   &CurlNoiseForce(Noise Strength=50, Pan Noise=true, Pan Noise Field=<0.5 0.5 0>)
   &ScaleSpriteSize(Scale Factor=&Vector2DFromFloat(Value=&FloatFromCurve(FloatCurve=DataInterface() {propertyValues {Curve {keys [
          <at 0 1 RCIM_Cubic>
          {arriveTangent=-1, interpMode=RCIM_Cubic, leaveTangent=-1, time=0.6, value=1.25}
          {arriveTangent=-0.1423, interpMode=RCIM_Cubic, leaveTangent=-0.1423, tangentMode=RCTM_User, time=1, value=0}
         ]}}})))
   &ScaleColor(
    Scale Alpha=&Multiply_Float(A=1, B=&FNTimeOfDayScalar(Night Scale=0.5, Morning Scale=0.4, Evening Scale=0.3))
    Scale RGB=<5 5 5>
    ScaleA=true
    ScaleRGB=true
   )
   &Drag(Drag=2, Use Linear Drag=true, Use Rotational Drag=false)
   &InheritSourceMovement(ApplyForce=false, ApplyPosition=true, ApplyPositionFalloff=false)
   &SolveForcesAndVelocity(Clamp Velocity=false, Limit Acceleration=false)
  ]
  rendererClasses [&NiagaraSpriteRendererProperties]
  renderers [
   {rendererClass=&NiagaraSpriteRendererProperties}
   {rendererClass=&NiagaraSpriteRendererProperties, rendererIndex=1}
  ]
 }
 Ribbon_L {
  emitterUpdateScript [
   &EmitterState(Life Cycle Mode=::NewEnumerator1, Loop Duration=1, UseLoopDelay=false)
   &SpawnBurst_Instantaneous(Use Spawn Probability=false, UseLoopCountLimit=false)
   &SpawnRate(SpawnRate=30, Use Spawn Probability=false)
  ]
  particleSpawnScript [
   &InitializeParticle(
    Color=<rgba 1.5 1.5 1.5 1.5>
    Lifetime=0.2
    Ribbon Width=15
    Ribbon Width Mode=::NewEnumerator1
    UsePositionOffset=false
    Write Material Random=true
   )
   &SkeletalMeshLocation(
    Bone / Socket Index=1
    MaskSpawnGroup=false
    Mesh Sampling Type=::NewEnumerator5
    Skeletal Mesh=$System.Skeletal Mesh
    Socket Sampling Mode=::NewEnumerator5
   )
  ]
  particleUpdateScript [
   &ParticleState()
   &InheritSourceMovement(ApplyForce=false, ApplyPosition=true, ApplyPositionFalloff=false)
   &SolveForcesAndVelocity(Clamp Velocity=false, Limit Acceleration=false)
   &ScaleRibbonWidth(Ribbon Width Scale=&FloatFromCurve(FloatCurve=DataInterface() {propertyValues {Curve {keys [<at 0 1>,<at 1 0.4>]}}}))
   &ScaleColor(
    Scale Alpha=&FloatFromCurve(
     FloatCurve=DataInterface() {propertyValues {Curve {keys [{arriveTangent=-3.0806, interpMode=RCIM_Cubic, leaveTangent=-3.0806, tangentMode=RCTM_User, time=0, value=1}, <at 1 0 RCIM_Cubic>]}}}
     Scale Curve=&Multiply_Float(
      A=&FN_PickFloatDependOnPlatform(TargetPlatformActive=1, TargetPlatformNoActive=0.7)
      B=&FNTimeOfDayScalar(Night Scale=0.6, Morning Scale=1, Evening Scale=1)
     )
    )
    Scale RGB=&MakeVectorFromLinearColorRGB(Color=&ColorFromCurve(ColorCurve=DataInterface() {propertyValues {
        AlphaCurve {keys []}
        BlueCurve {keys [<at 0.3816 0 RCIM_Cubic>,<at 0.4901 0>]}
        GreenCurve {keys [<at 0.3816 0.3 RCIM_Cubic>,<at 0.4901 0.6575>]}
        RedCurve {keys [<at 0.3816 1.5 RCIM_Cubic>,<at 0.4901 1>]}
       }}))
    ScaleA=true
    ScaleRGB=true
   )
  ]
  rendererClasses [&NiagaraRibbonRendererProperties]
  renderers [{rendererClass=&NiagaraRibbonRendererProperties}]
 }
 Ribbon_R {
  emitterUpdateScript [
   &EmitterState(Life Cycle Mode=::NewEnumerator1, Loop Duration=1, UseLoopDelay=false)
   &SpawnBurst_Instantaneous(Use Spawn Probability=false, UseLoopCountLimit=false)
   &SpawnRate(SpawnRate=30, Use Spawn Probability=false)
  ]
  particleSpawnScript [
   &InitializeParticle(
    Color=<rgba 1.5 1.5 1.5 1.5>
    Lifetime=0.3
    Ribbon Width=15
    Ribbon Width Mode=::NewEnumerator1
    UsePositionOffset=false
    Write Material Random=true
   )
   &SkeletalMeshLocation(
    Bone / Socket Index=2
    MaskSpawnGroup=false
    Mesh Sampling Type=::NewEnumerator5
    Skeletal Mesh=$System.Skeletal Mesh
    Socket Sampling Mode=::NewEnumerator5
   )
  ]
  particleUpdateScript [
   &ParticleState()
   &InheritSourceMovement(ApplyForce=false, ApplyPosition=true, ApplyPositionFalloff=false)
   &SolveForcesAndVelocity(Clamp Velocity=false, Limit Acceleration=false)
   &ScaleRibbonWidth(Ribbon Width Scale=&FloatFromCurve(FloatCurve=DataInterface() {propertyValues {Curve {keys [<at 0 1>,<at 1 0.5>]}}}))
   &ScaleColor(
    Scale Alpha=&FloatFromCurve(
     FloatCurve=DataInterface() {propertyValues {Curve {keys [{arriveTangent=-3.0806, interpMode=RCIM_Cubic, leaveTangent=-3.0806, tangentMode=RCTM_User, time=0, value=1}, <at 1 0 RCIM_Cubic>]}}}
     Scale Curve=&Multiply_Float(
      A=&FN_PickFloatDependOnPlatform(TargetPlatformActive=1, TargetPlatformNoActive=0.7)
      B=&FNTimeOfDayScalar(Night Scale=0.6, Morning Scale=1, Evening Scale=1)
     )
    )
    Scale RGB=&MakeVectorFromLinearColorRGB(Color=&ColorFromCurve(ColorCurve=DataInterface() {propertyValues {
        AlphaCurve {keys []}
        BlueCurve {keys [<at 0.372 0 RCIM_Cubic>,<at 0.4745 0>]}
        GreenCurve {keys [<at 0.372 0.3 RCIM_Cubic>,<at 0.4745 0.7083>]}
        RedCurve {keys [<at 0.372 1.5 RCIM_Cubic>,<at 0.4745 1>]}
       }}))
    ScaleA=true
    ScaleRGB=true
   )
  ]
  rendererClasses [&NiagaraRibbonRendererProperties]
  renderers [{rendererClass=&NiagaraRibbonRendererProperties}]
 }
 smoke {
  emitterUpdateScript [
   &EmitterState(Inactive Response=::NewEnumerator1, Life Cycle Mode=::NewEnumerator1, Loop Duration=1, UseLoopDelay=false)
   &SpawnRate(SpawnRate=10, Use Spawn Probability=false)
  ]
  particleSpawnScript [
   &PartitionParticles(Number Of Partitions=3)
   &InitializeParticle(
    Color=&MakeCustomLinearColorFromBool(
     Bool=&RandomBool()
     True Color=&MakeCustomLinearColorFromBool(
      Bool=&RandomBool()
      True Color=&MakeCustomLinearColorFromBool(Bool=&RandomBool(), True Color=<rgba 1 0.7384 0.7188 1>, False Color=<rgba 1 0.9346 0.6198 1>)
      False Color=&MakeCustomLinearColorFromBool(Bool=&RandomBool(), True Color=<rgba 0.75 1 0.7942 1>, False Color=<rgba 0.7344 1 0.9475 1>)
     )
     False Color=&MakeCustomLinearColorFromBool(
      Bool=&RandomBool()
      True Color=&MakeCustomLinearColorFromBool(Bool=&RandomBool(), True Color=<rgba 0.6458 0.6933 1 1>, False Color=<rgba 0.651 1 0.9517 1>)
      False Color=&MakeCustomLinearColorFromBool(Bool=&RandomBool(), True Color=<rgba 0.8673 0.6823 1 1>, False Color=<rgba 1 0.9463 0.6875 1>)
     )
    )
    Lifetime Max=1.5
    Lifetime Min=2
    Lifetime Mode=::NewEnumerator1
    Sprite Rotation Mode=::NewEnumerator2
    Sprite Size Mode=::NewEnumerator1
    Uniform Sprite Size Max=30
    Uniform Sprite Size Min=20
    UsePositionOffset=false
    Write Material Random=true
   )
   &ShapeLocation(
    MaskSpawnGroup=false
    Non Uniform Scale=<1 1 3>
    Offset=<0 5 0>
    Offset Mode=::NewEnumerator0
    Override Seed=false
    Sphere Radius=3
    UseOwnerScale=true
   )
   &CameraOffset(Camera Offset Amount=10)
   &SubUVAnimation(
    End Frame Range Override=1
    "Sprite Renderer "=DataInterface() {propertyValues {SpriteRenderer=&self:smoke_0.NiagaraSpriteRendererProperties_1}}
    SubUV Animation Mode:ENiagara_SubUVLookupModeV2=::NewEnumerator2
    UseEndFrame=true
    UseStartFrame=true
   )
  ]
  particleUpdateScript [
   &ParticleState()
   &SkeletalMeshLocation(
    Bone / Socket Index=&SelectIntFromArray(Direct Array Index=$Particles.PartitionParticles.Partition, Int Selection Array=DataInterface() {propertyValues {IntData [1,2]}})
    MaskSpawnGroup=false
    Mesh Sampling Type=::NewEnumerator5
    Sampled Position Offset=&AddVectorToPosition(Position=<0 -4 0>, Vector=$Particles.ShapeLocation.ShapeVector)
    Skeletal Mesh=$System.Skeletal Mesh
    Socket Sampling Mode=::NewEnumerator5
   )
   &ScaleSpriteSize(Uniform Curve Sprite Scale=DataInterface() {propertyValues {Curve {keys [{arriveTangent=0.4032, time=0, value=0.8}, <at 1 1>]}}})
   &SpriteRotationRate(Rotation Rate=&RandomRangeFloat(Minimum=-50, Maximum=50))
   &ScaleColor(
    Scale Alpha=&Multiply_Float(
     A=&FloatFromCurve(
      FloatCurve=DataInterface() {propertyValues {Curve {keys [
          {interpMode=RCIM_Cubic, tangentMode=RCTM_User, time=0, value=0}
          {interpMode=RCIM_Cubic, tangentMode=RCTM_User, time=0.3, value=1}
          {tangentMode=RCTM_User, time=1, value=0}
         ]}}}
      Scale Curve=&FN_PickFloatDependOnPlatform(TargetPlatformActive=1, TargetPlatformNoActive=0.15)
     )
     B=&Multiply_Float(
      A=&FNTimeOfDayScalar(Night Scale=0.4, Morning Scale=0.15, Evening Scale=0.3)
      B=&Multiply_Float(
       A=&MakeCustomFloatFromBool(Bool=$User.bIsFrontEndPreview, True Float=0.2, False Float=1)
       B=&FloatFromCurve(FloatCurve=DataInterface() {propertyValues {Curve {keys [
            {interpMode=RCIM_Cubic, tangentMode=RCTM_User, time=0, value=0}
            <at 0.4 0.45 RCIM_Cubic>
            {interpMode=RCIM_Cubic, tangentMode=RCTM_User, time=1, value=0}
           ]}}})
      )
     )
    )
    Scale RGB=&MakeVectorFromLinearColorRGB(Color=<rgba 3 1.3017 0.6406 1>)
    ScaleA=true
    ScaleRGB=true
   )
   &InheritSourceMovement(ApplyForce=false, ApplyPosition=true, ApplyPositionFalloff=true, ApplyPositionFalloffCurve=false) {enabled=false}
   &SolveForcesAndVelocity(Clamp Velocity=false, Limit Acceleration=false)
  ]
  rendererClasses [&NiagaraSpriteRendererProperties]
  renderers [{rendererClass=&NiagaraSpriteRendererProperties}]
 }
}
summary {
 systemName=NS_Emote_Phantom_HandTrail
 emitters [
  {emitterName=smoke, rendererClasses [&NiagaraSpriteRendererProperties]}
  {emitterName=Particle, rendererClasses [&NiagaraSpriteRendererProperties]}
  {emitterName=Ribbon_L, rendererClasses [&NiagaraRibbonRendererProperties]}
  {emitterName=Ribbon_R, rendererClasses [&NiagaraRibbonRendererProperties]}
 ]
 userVariables [
  User.bIsFrontEnd {type=NiagaraBool}
  User.bIsFrontEndPreview {type=NiagaraBool}
 ]
}
