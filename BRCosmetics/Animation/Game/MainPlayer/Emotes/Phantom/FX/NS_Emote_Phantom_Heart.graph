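# Niagara system for the heart effect of the Phantom emote, a short pink-red burst made of four burst-spawned emitters.
# Heat_mesh expands a mesh with curve-driven dynamic material parameters, Rays001 grows radial-material sprites, Sparks scatters pulsing glow sprites, and glow adds shrinking sprites plus a brief light.
# Time-of-day scalars adjust brightness and alpha, while front-end preview switches adjust scale and intensity and zero the light radius.
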NiagaraSystem /BRCosmetics/Animation/Game/MainPlayer/Emotes/Phantom/FX/NS_Emote_Phantom_Heart
dependencies {
 usedDataInterfaces [
  &NiagaraDataInterfaceCurve
  &NiagaraDataInterfacePlatformSet
  &NiagaraDataInterfaceColorCurve
  &NiagaraDataInterfaceVector2DCurve
 ]
 usedDynamicInputs []
 usedModules [&SystemState]
}
emitters {
 Heat_mesh {
  emitterUpdateScript [
   &EmitterState(Inactive Response=::NewEnumerator1, Life Cycle Mode=::NewEnumerator1, Loop Behavior=::NewEnumerator1, Loop Duration=1, UseLoopDelay=false)
   &SpawnBurst_Instantaneous(Use Spawn Probability=false, UseLoopCountLimit=false)
  ]
  particleSpawnScript [
   &InitializeParticle(
    Color=<rgba 3 0.6 0.7558 0.3>
    Lifetime=0.4
    Mesh Scale Mode=::NewEnumerator3
    Mesh Uniform Scale=2.2
    UsePositionOffset=true
    Write Material Random=true
   )
   &AddVelocity(Velocity=<50 0 0>)
  ]
  particleUpdateScript [
   &ParticleState()
   &ScaleMeshSize(Scale Factor=&Multiply_Vector(A=<1 1 1>, B=<1 1 1>)) {enabled=false, module=ScaleMeshSize001}
   &DynamicMaterialParameters(
    Index 0 Param 1=&FloatFromCurve(
     FloatCurve=DataInterface() {propertyValues {Curve {keys [<at 0 12>,<at 1 4>]}}}
     Scale Curve=&MakeCustomFloatFromBool(Bool=$User.bIsFrontEndPreview, True Float=0.9, False Float=1.1)
    )
    Index 0 Param 2=&FloatFromCurve(FloatCurve=DataInterface() {propertyValues {Curve {keys [
         {arriveTangent=-0.0886, interpMode=RCIM_Cubic, leaveTangent=-0.0886, tangentMode=RCTM_User, time=0, value=0}
         {arriveTangent=0.9995, interpMode=RCIM_Cubic, leaveTangent=0.9995, tangentMode=RCTM_User, time=0.4, value=0.5}
         {arriveTangent=9.8708, interpMode=RCIM_Cubic, leaveTangent=9.8708, tangentMode=RCTM_User, time=1, value=4.5}
        ]}}})
    Index 0 Param 3=2.2
    Index 0 Param 4=1
    Module.Param0WriteEnabled=true
    Module.Param1WriteEnabled=true
    Module.Param2WriteEnabled=true
    Module.Param3WriteEnabled=true
   )
   &ScaleColor(
    Scale Alpha=&FloatFromCurve(
     FloatCurve=DataInterface() {propertyValues {Curve {keys [
         {arriveTangent=9.0725, interpMode=RCIM_Cubic, leaveTangent=9.0725, tangentMode=RCTM_User, time=0.2, value=0}
         {arriveTangent=0.0702, interpMode=RCIM_Cubic, leaveTangent=0.0702, tangentMode=RCTM_User, time=0.4, value=1}
         {arriveTangent=-1.518, interpMode=RCIM_Cubic, leaveTangent=-1.518, tangentMode=RCTM_User, time=0.6, value=0.4}
         {arriveTangent=-0.0405, interpMode=RCIM_Cubic, leaveTangent=-0.0405, tangentMode=RCTM_User, time=1, value=0}
        ]}}}
     Scale Curve=&Multiply_Float(A=1, B=&FNTimeOfDayScalar(Night Scale=0.5, Morning Scale=0.8, Evening Scale=1))
    )
    Scale RGB=&VectorFromFloat(Value=1)
    ScaleA=true
    ScaleRGB=true
   )
   &ScaleMeshSize(Scale Factor=&VectorFromFloat(Value=&FloatFromCurve(
      FloatCurve=DataInterface() {propertyValues {Curve {keys [
          {arriveTangent=1.5111, interpMode=RCIM_Cubic, leaveTangent=1.5111, tangentMode=RCTM_User, time=0, value=0.6}
          {arriveTangent=0.3852, interpMode=RCIM_Cubic, leaveTangent=0.3852, tangentMode=RCTM_User, time=1, value=1.4}
         ]}}}
      Scale Curve=0.68
     )))
   &Drag(Drag=2, Use Linear Drag=true, Use Rotational Drag=true)
   &SolveForcesAndVelocity(Clamp Velocity=false, Limit Acceleration=false)
   /SetVariables(Particles.Color003:LinearColor=&MakeLinearColorFromVectorAndFloat(
     "Vector (RGB)"=&MakeVectorFromLinearColorRGB(Color=<rgba 2 0.36 0.4665 1>)
     "Float (Alpha)"=&FloatFromCurve(
      FloatCurve=DataInterface() {propertyValues {Curve {keys [
          {arriveTangent=9.0725, interpMode=RCIM_Cubic, leaveTangent=9.0725, tangentMode=RCTM_User, time=0, value=0}
          {arriveTangent=0.0702, interpMode=RCIM_Cubic, leaveTangent=0.0702, tangentMode=RCTM_User, time=0.1, value=1}
          {arriveTangent=-3.5329, interpMode=RCIM_Cubic, leaveTangent=-3.5329, tangentMode=RCTM_User, time=0.3, value=0.4}
          {arriveTangent=-0.0405, interpMode=RCIM_Cubic, leaveTangent=-0.0405, tangentMode=RCTM_User, time=0.6, value=0}
         ]}}}
      Scale Curve=&Multiply_Float(A=2, B=&FNTimeOfDayScalar(Night Scale=0.8, Morning Scale=1, Evening Scale=1))
     )
    )) {isSetParameters=true}
   /SetVariables(Particles.Size003:Vector3f=&Multiply_Vector(
     A=<1 1 1>
     B=&VectorFromFloat(Value=&FloatFromCurve(
       FloatCurve=DataInterface() {propertyValues {Curve {keys [
           {arriveTangent=2.1663, interpMode=RCIM_Cubic, leaveTangent=2.1663, tangentMode=RCTM_User, time=0, value=0.3}
           {arriveTangent=0.95, interpMode=RCIM_Cubic, leaveTangent=0.95, time=0.3, value=0.9}
           {arriveTangent=0.0212, interpMode=RCIM_Cubic, leaveTangent=0.0212, tangentMode=RCTM_User, time=1, value=1.25}
          ]}}}
       Scale Curve=0.65
      ))
    )) {isSetParameters=true}
   /SetVariables(Particles.Position003:NiagaraPosition=&Add_Vector(A=Linked() {linkedVariable {name=Particles.Position, type=NiagaraPosition}}, B=<0 0 4>)) {isSetParameters=true}
  ]
  rendererClasses [&NiagaraMeshRendererProperties]
  renderers [
   {rendererClass=&NiagaraMeshRendererProperties, rendererIndex=0}
   {rendererClass=&NiagaraMeshRendererProperties, rendererIndex=1}
  ]
 }
 Rays001 {
  emitterUpdateScript [
   &EmitterState(Life Cycle Mode=::NewEnumerator1, Loop Behavior=::NewEnumerator1, Loop Duration=1, UseLoopDelay=false)
   &SpawnBurst_Instantaneous(Spawn Count=2, Use Spawn Probability=false, UseLoopCountLimit=false)
   &SpawnRate(SpawnRate=2, Use Spawn Probability=false) {enabled=false}
  ]
  particleSpawnScript [
   &InitializeParticle(
    Lifetime=0.2
    Position Offset=<0 0 4>
    Sprite Rotation Mode=::NewEnumerator2
    Sprite Size Mode=::NewEnumerator3
    Uniform Sprite Size=50
    UsePositionOffset=true
    Write Material Random=true
   )
   &AddVelocity(Velocity=<50 0 0>)
  ]
  particleUpdateScript [
   &ParticleState()
   &ApplyOwnerScaleToAttributes(Scale Initial Sprite Size=true)
   &ScaleSpriteSize(Uniform Curve Sprite Scale=DataInterface() {propertyValues {Curve {keys [{interpMode=RCIM_Cubic, tangentMode=RCTM_User, time=0, value=0.7}, {interpMode=RCIM_Cubic, tangentMode=RCTM_User, time=1, value=1.8}]}}})
   &FN_Radial_Material(Base Sprite Tile U=0.5, Base Sprite Tile V=2, Radial Mask Power=3, Radial Mask Strength=0.8)
   &ScaleColor(
    Scale Alpha=&FloatFromCurve(
     FloatCurve=DataInterface() {propertyValues {Curve {keys [{interpMode=RCIM_Cubic, tangentMode=RCTM_User, time=0, value=1}, {interpMode=RCIM_Cubic, tangentMode=RCTM_User, time=1, value=0}]}}}
     Scale Curve=&Multiply_Float(A=1.2, B=&FNTimeOfDayScalar(Night Scale=0.5, Morning Scale=1, Evening Scale=1))
    )
    Scale RGB=&MakeVectorFromLinearColorRGB(Color=<rgba 1.5 0.3 0.3779 1>)
    ScaleA=true
    ScaleRGB=true
   )
   &Drag(Drag=2, Use Linear Drag=true, Use Rotational Drag=true)
   &SolveForcesAndVelocity(Clamp Velocity=false, Limit Acceleration=false)
  ]
  rendererClasses [&NiagaraSpriteRendererProperties]
  renderers [{rendererClass=&NiagaraSpriteRendererProperties}]
 }
 Sparks {
  emitterUpdateScript [
   &EmitterState(Inactive Response=::NewEnumerator1, Life Cycle Mode=::NewEnumerator1, Loop Behavior=::NewEnumerator1, Loop Duration=1, UseLoopDelay=false) {module=EmitterState001}
   &SpawnBurst_Instantaneous(Spawn Count=6, Use Spawn Probability=false, UseLoopCountLimit=false)
  ]
  particleSpawnScript [
   &InitializeParticle(
    Color Maximum=<rgba 1.5 0.15 0.317 1>
    Color Minimum=<rgba 1 0.1 0.1358 1>
    Color Mode=::NewEnumerator2
    Lifetime Max=0.4
    Lifetime Min=0.2
    Lifetime Mode=::NewEnumerator1
    Sprite Size Mode=::NewEnumerator1
    Uniform Sprite Size Max=15
    Uniform Sprite Size Min=10
    UsePositionOffset=false
    Write Material Random=true
   ) {module=InitializeParticle001}
   &ShapeLocation(
    MaskSpawnGroup=false
    Offset=<10 0 5>
    Offset Mode=::NewEnumerator0
    Override Seed=false
    Shape Origin=$Particles.Position
    Sphere Radius=20
    UseOwnerScale=true
   )
   &CameraOffset(Camera Offset Amount=&FN_FrontEndFloatModifier(Gameplay=20, Front End=0)) {enabled=false}
   &AddVelocity(Module.Use Falloff Distance=false, Origin Offset=<5 0 5>, Override Seed=false, Velocity Mode=::NewEnumerator1, Velocity Speed=120)
  ]
  particleUpdateScript [
   &ParticleState()
   &Drag(Drag=5, Use Linear Drag=true, Use Rotational Drag=false)
   &SolveForcesAndVelocity(Clamp Velocity=false, Limit Acceleration=false)
   &ScaleColor(
    Scale Alpha=&FloatFromCurve(
     FloatCurve=DataInterface() {propertyValues {Curve {keys [
         {interpMode=RCIM_Cubic, tangentMode=RCTM_User, time=0, value=0}
         {arriveTangent=-0.8637, interpMode=RCIM_Cubic, leaveTangent=-0.8637, tangentMode=RCTM_User, time=0.3, value=1}
         {arriveTangent=0.0103, interpMode=RCIM_Cubic, leaveTangent=0.0103, tangentMode=RCTM_User, time=0.8, value=0}
        ]}}}
     CurveIndex=$Emitter.NormalizedLoopAge
    )
    Scale RGB=&VectorFromFloat(Value=&FNTimeOfDayScalar(Night Scale=2, Morning Scale=3, Evening Scale=3))
    ScaleA=true
    ScaleRGB=true
   )
   &ScaleSpriteSize(
    Uniform Curve Index=$Emitter.NormalizedLoopAge
    Uniform Curve Scale=&FN_FrontEndFloatModifier(Gameplay=1, Front End=0.5)
    Uniform Curve Sprite Scale=DataInterface() {propertyValues {Curve {keys [
        {arriveTangent=4.1785, interpMode=RCIM_Cubic, leaveTangent=4.1785, tangentMode=RCTM_User, time=0, value=0}
        {arriveTangent=0.9995, interpMode=RCIM_Cubic, leaveTangent=0.9995, tangentMode=RCTM_User, time=0.1, value=0.9}
        {arriveTangent=-0.6667, interpMode=RCIM_Cubic, leaveTangent=-0.6667, time=0.3, value=0.2}
        {arriveTangent=-0.25, interpMode=RCIM_Cubic, leaveTangent=-0.25, time=0.4, value=0.7}
        {arriveTangent=-0.25, interpMode=RCIM_Cubic, leaveTangent=-0.25, time=0.5, value=0.15}
        {arriveTangent=-0.25, interpMode=RCIM_Cubic, leaveTangent=-0.25, time=0.6, value=0.65}
        {arriveTangent=-0.6667, interpMode=RCIM_Cubic, leaveTangent=-0.6667, time=0.7, value=0.1}
        {arriveTangent=-0.3333, interpMode=RCIM_Cubic, leaveTangent=-0.3333, time=0.9, value=0.45}
        {arriveTangent=-3.4484, interpMode=RCIM_Cubic, leaveTangent=-3.4484, tangentMode=RCTM_User, time=1, value=0}
       ]}}}
   )
   /SetVariables(Particles.Glow Color=&MakeLinearColorFromVectorAndFloat(
     "Vector (RGB)"=&MakeVectorFromLinearColorRGB(Color=$Particles.Color)
     "Float (Alpha)"=&FloatFromCurve(
      FloatCurve=DataInterface() {propertyValues {Curve {keys [
          {interpMode=RCIM_Cubic, tangentMode=RCTM_User, time=0, value=0}
          {arriveTangent=-0.8637, interpMode=RCIM_Cubic, leaveTangent=-0.8637, tangentMode=RCTM_User, time=0.3, value=1}
          {arriveTangent=0.0103, interpMode=RCIM_Cubic, leaveTangent=0.0103, tangentMode=RCTM_User, time=0.8, value=0}
         ]}}}
      CurveIndex=$Emitter.NormalizedLoopAge
      Scale Curve=&Multiply_Float(A=0.1, B=&FNTimeOfDayScalar(Night Scale=0.5, Morning Scale=0.5, Evening Scale=0.6))
     )
    )) {isSetParameters=true}
   /SetVariables(Particles.GlowSize=&Multiply_Vector2D(A=$Particles.SpriteSize, B=&Vector2DFromFloat(Value=2))) {isSetParameters=true}
  ]
  rendererClasses [&NiagaraSpriteRendererProperties]
  renderers [
   {rendererClass=&NiagaraSpriteRendererProperties}
   {rendererClass=&NiagaraSpriteRendererProperties, rendererIndex=1}
  ]
 }
 glow {
  emitterUpdateScript [
   &EmitterState(Inactive Response=::NewEnumerator1, Life Cycle Mode=::NewEnumerator1, Loop Behavior=::NewEnumerator1, Loop Duration=0.5, UseLoopDelay=false)
   &SpawnBurst_Instantaneous(Use Spawn Probability=false, UseLoopCountLimit=false)
  ]
  particleSpawnScript [
   &InitializeParticle(
    Color=<rgba 0.5 0.5 0.5 1>
    Lifetime=0.35
    Position Offset=<0 0 5>
    Sprite Size Mode=::NewEnumerator3
    Uniform Sprite Size=150
    UsePositionOffset=true
    Write Material Random=true
   )
   &CameraOffset(Camera Offset Amount=&Add_Float(
     A=20
     B=&FN_PickFloatDependOnPlatform(
      TargetPlatformActive=1
      TargetPlatformNoActive=&MakeCustomFloatFromBool(Bool=$User.bIsFrontEndPreview, True Float=25, False Float=100)
     )
    ))
   &AddVelocity(Velocity=<50 0 0>)
  ]
  particleUpdateScript [
   &ParticleState()
   &ScaleSpriteSize(
    Uniform Curve Scale=&MakeCustomFloatFromBool(Bool=$User.bIsFrontEndPreview, True Float=0.5, False Float=1)
    Uniform Curve Sprite Scale=DataInterface() {propertyValues {Curve {keys [<at 0 1>, {tangentMode=RCTM_User, time=1, value=0.2}]}}}
   )
   /SetVariables(Particles.Size002:Vector2f=&Multiply_Vector2D(A=$Particles.SpriteSize, B=&Vector2DFromFloat(Value=0.6))) {isSetParameters=true}
   &ScaleColor(
    Scale Alpha=&FloatFromCurve(
     FloatCurve=DataInterface() {propertyValues {Curve {keys [
         {arriveTangent=8.4201, interpMode=RCIM_Cubic, leaveTangent=8.4201, tangentMode=RCTM_User, time=0, value=-0.025}
         {arriveTangent=0.2873, interpMode=RCIM_Cubic, leaveTangent=-0.1265, tangentMode=RCTM_Break, time=0.2, value=0.8}
         {arriveTangent=-2.3949, interpMode=RCIM_Cubic, leaveTangent=-2.3949, tangentMode=RCTM_User, time=0.5, value=0.4}
         <at 1 0 RCIM_Cubic>
        ]}}}
     Scale Curve=&Multiply_Float(
      A=&FNTimeOfDayScalar(Night Scale=0.7, Morning Scale=1, Evening Scale=1)
      B=&MakeCustomFloatFromBool(Bool=$User.bIsFrontEndPreview, True Float=0.5, False Float=1)
     )
    )
    Scale RGB=&MakeVectorFromLinearColorRGB(Color=&ColorFromCurve(ColorCurve=DataInterface() {propertyValues {
        AlphaCurve {keys []}
        BlueCurve {keys [<at 0.7006 0.06 RCIM_Cubic>]}
        GreenCurve {keys [<at 0.7006 0.025 RCIM_Cubic>]}
        RedCurve {keys [<at 0.7006 0.3 RCIM_Cubic>]}
       }}))
    ScaleA=true
    ScaleRGB=true
   )
   /SetVariables(Particles.color002:LinearColor=&MakeLinearColorFromVectorAndFloat(
     "Vector (RGB)"=&MakeVectorFromLinearColorRGB(Color=<rgba 1 0.1354 0.1563 1>)
     "Float (Alpha)"=&FloatFromCurve(
      FloatCurve=DataInterface() {propertyValues {Curve {keys [
          {arriveTangent=-1.1243, interpMode=RCIM_Cubic, leaveTangent=-1.1243, tangentMode=RCTM_User, time=0, value=1}
          {arriveTangent=-2.3022, interpMode=RCIM_Cubic, leaveTangent=-2.3022, tangentMode=RCTM_User, time=0.2371, value=0.653}
          {arriveTangent=-0.0759, interpMode=RCIM_Cubic, leaveTangent=-0.0759, tangentMode=RCTM_User, time=0.6, value=0}
         ]}}}
      Scale Curve=&Multiply_Float(
       A=&FNTimeOfDayScalar(Night Scale=0.5, Morning Scale=0.7, Evening Scale=0.7)
       B=&FN_FrontEndFloatModifier(Gameplay=1, Front End=0.8)
      )
     )
    )) {isSetParameters=true}
   &FN_Light_Attributes(
    Color=&ColorFromCurve(ColorCurve=DataInterface() {propertyValues {AlphaCurve {keys []}, BlueCurve {keys [<at 0.1896 0.028>]}, GreenCurve {keys [<at 0.1896 0.021>]}, RedCurve {keys [<at 0.1896 1.5>]}}})
    Create Color Attribute=true
    Create Exponent Attribute=true
    Create Position Attribute=true
    Create Radius Attribute=true
    Exponent=2
    Radius=&FloatFromCurve(
     FloatCurve=DataInterface() {propertyValues {Curve {keys [<at 0 0.4>,<at 0.2 1.1>,<at 1 1.4>]}}}
     Scale Curve=&MakeCustomFloatFromBool(Bool=$User.bIsFrontEndPreview, True Float=0, False Float=100)
    )
    Scale Color=&FloatFromCurve(
     FloatCurve=DataInterface() {propertyValues {Curve {keys [
         {arriveTangent=4.1988, interpMode=RCIM_Cubic, leaveTangent=4.1988, tangentMode=RCTM_User, time=0, value=0}
         {arriveTangent=-0.0105, interpMode=RCIM_Cubic, leaveTangent=-0.0105, tangentMode=RCTM_User, time=0.15, value=1}
         {arriveTangent=-0.9284, interpMode=RCIM_Cubic, leaveTangent=-0.9284, tangentMode=RCTM_User, time=0.5, value=0}
        ]}}}
     Scale Curve=&Multiply_Float(A=&FNTimeOfDayScalar(Night Scale=0.5, Morning Scale=1, Evening Scale=1), B=20)
    )
   )
   &Drag(Drag=2, Use Linear Drag=true, Use Rotational Drag=true)
   &SolveForcesAndVelocity(Clamp Velocity=false, Limit Acceleration=false)
  ]
  rendererClasses [&NiagaraSpriteRendererProperties, &NiagaraLightRendererProperties]
  renderers [
   {rendererClass=&NiagaraSpriteRendererProperties}
   {rendererClass=&NiagaraSpriteRendererProperties, rendererIndex=1}
   {rendererClass=&NiagaraLightRendererProperties, rendererIndex=2}
  ]
 }
}
summary {
 systemName=NS_Emote_Phantom_Heart
 emitters [
  {emitterName=Heat_mesh, rendererClasses [&NiagaraMeshRendererProperties]}
  {emitterName=glow, rendererClasses [&NiagaraSpriteRendererProperties, &NiagaraLightRendererProperties]}
  {emitterName=Rays001, rendererClasses [&NiagaraSpriteRendererProperties]}
  {emitterName=Sparks, rendererClasses [&NiagaraSpriteRendererProperties]}
 ]
 userVariables [
  User.bIsFrontEnd {type=NiagaraBool}
  User.bIsFrontEndPreview {type=NiagaraBool}
 ]
}
