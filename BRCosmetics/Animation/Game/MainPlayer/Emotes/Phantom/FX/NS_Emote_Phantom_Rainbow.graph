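# Niagara effect for the Phantom emote that draws a rainbow mesh with a trail of multicolored sparkles.
# Rainbow animates a dynamic material parameter over its life, then fades, with offset and alpha adjusted for front-end preview.
# Particle_head follows a curve-driven arc, and Particle001 spawns from it via particle reads as randomly colored, pulsing sprites drifting on curl noise.
# Time-of-day scalars adjust the mesh and sparkle alpha.
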NiagaraSystem /BRCosmetics/Animation/Game/MainPlayer/Emotes/Phantom/FX/NS_Emote_Phantom_Rainbow
dependencies {
 usedDataInterfaces [&NiagaraDataInterfaceCurve, &NiagaraDataInterfaceParticleRead]
 usedDynamicInputs []
 usedModules [&SystemState]
}
emitters {
 Particle001 {
  emitterUpdateScript [
   &EmitterState(Life Cycle Mode=::NewEnumerator1, Loop Behavior=::NewEnumerator1, Loop Delay=0.1, Loop Duration=0.6, UseLoopDelay=true)
   &SpawnParticlesFromOtherEmitter(
    Attribute Reader=DataInterface() {propertyValues {EmitterBinding {bindingMode=Other, emitterName=Particle_head}}}
    Spawn Rate=60
    Use Maximum Age=false
    Use Minimum Age=false
   )
  ]
  particleSpawnScript [
   &InitializeParticle(
    Color=&MakeCustomLinearColorFromBool(
     Bool=&RandomBool()
     True Color=&MakeCustomLinearColorFromBool(
      Bool=&RandomBool()
      True Color=&MakeCustomLinearColorFromBool(Bool=&RandomBool(), True Color=<rgba 1 0.07 0 1>, False Color=<rgba 1 0.8281 0 1>)
      False Color=&MakeCustomLinearColorFromBool(Bool=&RandomBool(), True Color=<rgba 0 1 0.1767 1>, False Color=<rgba 0 1 0.8022 1>)
     )
     False Color=&MakeCustomLinearColorFromBool(
      Bool=&RandomBool()
      True Color=&MakeCustomLinearColorFromBool(Bool=&RandomBool(), True Color=<rgba 0 0.1339 1 1>, False Color=<rgba 0 1 0.8617 1>)
      False Color=&MakeCustomLinearColorFromBool(Bool=&RandomBool(), True Color=<rgba 0.5824 0 1 1>, False Color=<rgba 1 0.8281 0 1>)
     )
    )
    Lifetime=&RandomRangeFloat(Minimum=0.2, Maximum=1)
    Sprite Rotation Mode=::NewEnumerator2
    Sprite Size Mode=::NewEnumerator3
    Uniform Sprite Size=&FloatFromCurve(
     FloatCurve=DataInterface() {propertyValues {Curve {keys [<at 0.8 1>,<at 0.9 4>]}}}
     CurveIndex=&RandomRangeFloat(Minimum=0.5, Maximum=1)
    )
    UsePositionOffset=false
    Write Material Random=true
   )
   &SampleParticlesFromOtherEmitter()
   /SetVariables(Particles.RandomCore:NiagaraFloat=&FloatFromCurve(
     FloatCurve=DataInterface() {propertyValues {Curve {keys [
         {arriveTangent=0.0559, interpMode=RCIM_Cubic, leaveTangent=0.0559, tangentMode=RCTM_User, time=0, value=0.2}
         {arriveTangent=0.1895, interpMode=RCIM_Cubic, leaveTangent=0.1895, tangentMode=RCTM_User, time=0.9, value=0.25}
         {arriveTangent=1.1808, interpMode=RCIM_Cubic, leaveTangent=1.1808, tangentMode=RCTM_User, time=1, value=0.3}
        ]}}}
     CurveIndex=&RandomRangeFloat(Minimum=0, Maximum=1)
     Scale Curve=2
    )) {isSetParameters=true}
   &ShapeLocation(MaskSpawnGroup=false, Offset Mode=::NewEnumerator0, Override Seed=false, Sphere Radius=16, UseOwnerScale=true)
  ]
  particleUpdateScript [
   &ParticleState()
   /SetVariables(Particles.Glow03_Color=&MakeLinearColorFromVectorAndFloat(
     "Vector (RGB)"=&MakeVectorFromLinearColorRGB(Color=&MakeCustomLinearColorFromBool(
       Bool=&RandomBool()
       True Color=&MakeCustomLinearColorFromBool(
        Bool=&RandomBool()
        True Color=&MakeCustomLinearColorFromBool(Bool=&RandomBool(), True Color=<rgba 1 0.07 0 1>, False Color=<rgba 1 0.8281 0 1>)
        False Color=&MakeCustomLinearColorFromBool(Bool=&RandomBool(), True Color=<rgba 0 1 0.1767 1>, False Color=<rgba 0 1 0.8022 1>)
       )
       False Color=&MakeCustomLinearColorFromBool(
        Bool=&RandomBool()
        True Color=&MakeCustomLinearColorFromBool(Bool=&RandomBool(), True Color=<rgba 0 0.1339 1 1>, False Color=<rgba 0 1 0.8617 1>)
        False Color=&MakeCustomLinearColorFromBool(Bool=&RandomBool(), True Color=<rgba 0.5824 0 1 1>, False Color=<rgba 1 0.8281 0 1>)
       )
      ))
     "Float (Alpha)"=&Multiply_Float(
      A=&Multiply_Float(
       A=&FloatFromCurve(FloatCurve=DataInterface() {propertyValues {Curve {keys [
            <at 0 1 RCIM_Cubic>
            <at 0.2 0.2 RCIM_Cubic>
            {arriveTangent=-0.25, interpMode=RCIM_Cubic, leaveTangent=-0.25, time=0.6, value=1}
            {arriveTangent=-0.1423, interpMode=RCIM_Cubic, leaveTangent=-0.1423, tangentMode=RCTM_User, time=1, value=0}
           ]}}})
       B=&FloatFromCurve(FloatCurve=DataInterface() {propertyValues {Curve {keys [<at 0 1>,<at 1 0>]}}}, CurveIndex=$Emitter.Age)
      )
      B=&FNTimeOfDayScalar(Night Scale=0.5, Morning Scale=0.1, Evening Scale=0.3)
     )
    )) {isSetParameters=true}
   /SetVariables(Particles.Glow03_Size=&Multiply_Vector2D_ByFloat(
     Float=&FloatFromCurve(
      FloatCurve=DataInterface() {propertyValues {Curve {keys [
          <at 0 0 RCIM_Cubic>
          {arriveTangent=-1.25, interpMode=RCIM_Cubic, leaveTangent=-1.25, time=0.1282, value=1.1909}
          {arriveTangent=-0.7181, interpMode=RCIM_Cubic, leaveTangent=-0.7181, time=0.3, value=-0.375}
          {arriveTangent=0.4167, interpMode=RCIM_Cubic, leaveTangent=0.4167, time=0.3338, value=1.0432}
          {arriveTangent=0.1533, interpMode=RCIM_Cubic, leaveTangent=0.1533, time=0.6, value=-0.25}
          {arriveTangent=0.625, interpMode=RCIM_Cubic, leaveTangent=0.625, time=0.7892, value=1.113}
          {arriveTangent=-0.1423, interpMode=RCIM_Cubic, leaveTangent=-0.1423, tangentMode=RCTM_User, time=1, value=0}
         ]}}}
      Scale Curve=10
     )
     Vector2D=<1 1>
    )) {isSetParameters=true}
   &CurlNoiseForce(Noise Strength=50, Pan Noise=true, Pan Noise Field=<0.5 0.5 0>)
   &ScaleSpriteSize(Scale Factor=&Vector2DFromFloat(Value=&FloatFromCurve(
      FloatCurve=DataInterface() {propertyValues {Curve {keys [
          <at 0 0 RCIM_Cubic>
          {arriveTangent=-1.0984, interpMode=RCIM_Cubic, leaveTangent=-1.0984, time=0.1282, value=1.1909}
          {arriveTangent=-0.7181, interpMode=RCIM_Cubic, leaveTangent=-0.7181, time=0.2922, value=-0.3209}
          {arriveTangent=0.3039, interpMode=RCIM_Cubic, leaveTangent=0.3039, time=0.3338, value=1.0432}
          {arriveTangent=-0.1259, interpMode=RCIM_Cubic, leaveTangent=-0.1259, time=0.6211, value=-0.2209}
          {arriveTangent=0.5832, interpMode=RCIM_Cubic, leaveTangent=0.5832, time=0.7892, value=0.9859}
          {arriveTangent=-0.1423, interpMode=RCIM_Cubic, leaveTangent=-0.1423, tangentMode=RCTM_User, time=1, value=0}
         ]}}}
      Scale Curve=1.5
     )))
   &ScaleColor(
    Scale Alpha=&Multiply_Float(
     A=&Multiply_Float(
      A=1
      B=&FloatFromCurve(FloatCurve=DataInterface() {propertyValues {Curve {keys [<at 0 1>,<at 1 0>]}}}, CurveIndex=$Emitter.Age)
     )
     B=&FNTimeOfDayScalar(Night Scale=0.5, Morning Scale=0.1, Evening Scale=0.3)
    )
    Scale RGB=<5 5 5>
    ScaleA=true
    ScaleRGB=true
   )
   &Drag(Drag=2, Use Linear Drag=true, Use Rotational Drag=false)
   &SolveForcesAndVelocity(Clamp Velocity=false, Limit Acceleration=false)
  ]
  rendererClasses [&NiagaraSpriteRendererProperties]
  renderers [
   {rendererClass=&NiagaraSpriteRendererProperties}
   {rendererClass=&NiagaraSpriteRendererProperties, rendererIndex=1}
  ]
 }
 Particle_head {
  emitterUpdateScript [
   &EmitterState(Inactive Response=::NewEnumerator1, Life Cycle Mode=::NewEnumerator1, Loop Behavior=::NewEnumerator1, Loop Duration=1, UseLoopDelay=false)
   &SpawnBurst_Instantaneous(Use Spawn Probability=false, UseLoopCountLimit=false)
  ]
  particleSpawnScript [&InitializeParticle(Sprite Size Mode=::NewEnumerator3, Uniform Sprite Size=20, UsePositionOffset=false, Write Material Random=true)]
  particleUpdateScript [
   &ParticleState()
   &SolveForcesAndVelocity(Clamp Velocity=false, Limit Acceleration=false)
   &ShapeLocation(
    MaskSpawnGroup=false
    Offset=&MakeVector(
     X=0
     Y=&FloatFromCurve(FloatCurve=DataInterface() {propertyValues {Curve {keys [<at 0.2 -70>,<at 0.7 66>]}}})
     Z=&FloatFromCurve(FloatCurve=DataInterface() {propertyValues {Curve {keys [
          {arriveTangent=108.4907, interpMode=RCIM_Cubic, leaveTangent=108.4907, tangentMode=RCTM_User, time=0, value=30}
          {arriveTangent=139.4345, interpMode=RCIM_Cubic, leaveTangent=139.4345, tangentMode=RCTM_User, time=0.1308, value=46.45}
          {arriveTangent=-124.043, interpMode=RCIM_Cubic, leaveTangent=-124.043, tangentMode=RCTM_User, time=0.5, value=95}
          {arriveTangent=-14.5404, interpMode=RCIM_Cubic, leaveTangent=-14.5404, tangentMode=RCTM_User, time=1, value=30}
         ]}}})
    )
    Offset Mode=::NewEnumerator0
    Override Seed=false
    Sphere Radius=0.1
    UseOwnerScale=true
   )
  ]
  rendererClasses [&NiagaraSpriteRendererProperties]
  renderers [{rendererClass=&NiagaraSpriteRendererProperties}]
 }
 Rainbow {
  emitterUpdateScript [
   &EmitterState(Inactive Response=::NewEnumerator1, Life Cycle Mode=::NewEnumerator1, Loop Behavior=::NewEnumerator1, UseLoopDelay=true)
   &SpawnBurst_Instantaneous(Use Spawn Probability=false, UseLoopCountLimit=false)
  ]
  particleSpawnScript [
   &InitializeParticle(
    Mesh Scale=&Multiply_Vector(A=<1 1 1>, B=&VectorFromFloat(Value=2))
    Mesh Scale Mode=::NewEnumerator4
    Position Offset=&MakeCustomVectorFromBool(Bool=$User.bIsFrontEndPreview, True Vector=<0 0 -5>, False Vector=<0 0 0>)
    UsePositionOffset=true
    Write Material Random=true
   )
   &InitialMeshOrientation(Mesh Orientation Mode=::NewEnumerator2, Rotation=<0 0 0.06>, Use Rotation Vector=true) {enabled=false}
  ]
  particleUpdateScript [
   &ParticleState()
   &SolveForcesAndVelocity(Clamp Velocity=false, Limit Acceleration=false)
   &DynamicMaterialParameters(
    Index 0 Param 2=&FloatFromCurve(FloatCurve=DataInterface() {propertyValues {Curve {keys [<at 0 0 RCIM_Cubic>, {arriveTangent=0.7692, interpMode=RCIM_Cubic, leaveTangent=0.7692, tangentMode=RCTM_User, time=1, value=1}]}}})
    Module.Index0Float0Write=true
    Module.Index0Float1Write=true
    Module.Index0Float2Write=true
    Module.Index0Float3Write=true
   )
   &ScaleColor(
    Scale Alpha=&FloatFromCurve(
     FloatCurve=DataInterface() {propertyValues {Curve {keys [<at 0 1>,<at 0.6 1>,<at 0.9 0>]}}}
     Scale Curve=&Multiply_Float(
      A=&MakeCustomFloatFromBool(Bool=$User.bIsFrontEndPreview, True Float=1, False Float=2)
      B=&FNTimeOfDayScalar(Night Scale=1, Morning Scale=1, Evening Scale=0.7)
     )
    )
    ScaleA=true
    ScaleRGB=true
   )
  ]
  rendererClasses [&NiagaraMeshRendererProperties]
  renderers [{rendererClass=&NiagaraMeshRendererProperties, rendererIndex=0}]
 }
}
summary {
 systemName=NS_Emote_Phantom_Rainbow
 emitters [
  {emitterName=Rainbow, rendererClasses [&NiagaraMeshRendererProperties]}
  {emitterName=Particle_head, rendererClasses [&NiagaraSpriteRendererProperties]}
  {emitterName=Particle001, rendererClasses [&NiagaraSpriteRendererProperties]}
 ]
 userVariables [
  User.bIsFrontEnd {type=NiagaraBool}
  User.bIsFrontEndPreview {type=NiagaraBool}
 ]
}
