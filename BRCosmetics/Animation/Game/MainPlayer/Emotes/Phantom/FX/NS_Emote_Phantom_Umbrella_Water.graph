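# Niagara water effect for the Phantom umbrella emote, with two looping sprite emitters placed on a skeletal mesh using shape offsets.
# Drop spawns at a curve-tapered rate, with night-boosted brightness; its owner-speed check has no effect because both branches equal 1.
# Water fires periodic bursts with random velocity, drag and gravity, scaling sprite size differently for front end and gameplay.
# Both use SubUV and dynamic material parameters, and inherit source movement.
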NiagaraSystem /BRCosmetics/Animation/Game/MainPlayer/Emotes/Phantom/FX/NS_Emote_Phantom_Umbrella_Water
dependencies {
 usedDataInterfaces [&NiagaraDataInterfaceCurve, &NiagaraDataInterfaceVectorCurve, &NiagaraDataInterfaceSkeletalMesh]
 usedDynamicInputs []
 usedModules [
  &self:SystemSpawnScript.SystemScriptSource.SystemScriptGraph.NiagaraNodeAssignment_0.SetVariables_#1
  &SystemState
 ]
}
emitters {
 Drop {
  emitterUpdateScript [
   &EmitterState(
    Inactive Response=::NewEnumerator1
    Life Cycle Mode=::NewEnumerator1
    Loop Behavior=::NewEnumerator1
    Loop Delay=1
    Loop Duration=2
    UseLoopDelay=true
   )
   &SpawnRate(
    SpawnRate=&FloatFromCurve(
     FloatCurve=DataInterface() {propertyValues {Curve {keys [<at 0 1>,<at 0.6 1>,<at 1 0>]}}}
     CurveIndex=$Emitter.NormalizedLoopAge
     Scale Curve=&Multiply_Float(
      A=20
      B=&MakeCustomFloatFromBool(
       Bool=&SetBoolByFloatComparison(A=&VectorLength(Vector=$Engine.Owner.Velocity), B=20)
       True Float=1
       False Float=1
      )
     )
    )
    Use Spawn Probability=false
   )
  ]
  particleSpawnScript [
   &InitializeParticle(
    Color=<rgba 1.8333 2.7361 4 1>
    Lifetime Max=1.2
    Lifetime Mode=::NewEnumerator1
    Position Offset=<30 0 0>
    Position Offset Coordinate Space=::Local
    Ribbon Width=3
    Ribbon Width Mode=::NewEnumerator1
    Sprite Rotation Mode=::NewEnumerator2
    Sprite Size Max=<10 30>
    Sprite Size Min=<5 20>
    Sprite Size Mode=::NewEnumerator2
    UsePositionOffset=true
    Write Material Random=true
   )
   &ShapeLocation(
    Handle Radius=10
    Large Radius=40
    MaskSpawnGroup=false
    Offset=<20 0 0>
    Offset Mode=::NewEnumerator0
    Override Seed=false
    Rotation Angle=1
    Rotation Axis=<0 0 0>
    Rotation Mode=::NewEnumerator1
    Shape Primitive=::NewEnumerator3
    UseOwnerScale=true
   )
   &SkeletalMeshLocation(
    Bone Sampling Mode=::NewEnumerator4
    MaskSpawnGroup=false
    Orientation Sampling=::NewEnumerator1
    Sampled Position Offset=&AddVectorToPosition(Position=<-30 0 0>, Vector=$Particles.ShapeLocation.ShapeVector)
    Skeletal Mesh=$System.Skeletal Mesh
   )
   /Niagara/Modules/Update/SubUV/SubUVAnimation(Number Of Frames=8, SubUV Animation Mode=::NewEnumerator1) {module=SubUVAnimation001}
  ]
  particleUpdateScript [
   &ParticleState()
   &SpriteRotationRate(Rotation Rate=&RandomRangeFloat(Minimum=-20, Maximum=20))
   &Drag(Use Linear Drag=true, Use Rotational Drag=false)
   &GravityForce()
   &InheritSourceMovement(ApplyForce=false, ApplyPosition=true, ApplyPositionFalloff=false)
   &SolveForcesAndVelocity(Clamp Velocity=false, Limit Acceleration=false)
   &ScaleColor(
    Scale Alpha=&FloatFromCurve(
     FloatCurve=DataInterface() {propertyValues {Curve {keys [<at 0 -1>, {interpMode=RCIM_Cubic, tangentMode=RCTM_User, time=0.009, value=0.9316}, <at 0.7 0.875>, <at 1 -1>]}}}
     Scale Curve=&RandomRangeFloat(Minimum=0.25, Maximum=0.4)
    )
    Scale RGB=&VectorFromCurve(
     VectorCurve=DataInterface() {propertyValues {XCurve {keys [<at 0 2>,<at 1 0>]}, YCurve {keys [<at 0 2>,<at 1 0>]}, ZCurve {keys [<at 0 2>,<at 1 0>]}}}
     Scale Curve=&Multiply_VectorByFloat(Vector=<1 1 1>, Float=&FNTimeOfDayScalar(Night Scale=1.5, Morning Scale=1, Evening Scale=1))
    )
    ScaleA=true
    ScaleRGB=true
   ) {module=ScaleColor001}
   &ScaleSpriteSize(Scale Factor=&Vector2DFromFloat(Value=&FloatFromCurve(
      FloatCurve=DataInterface() {propertyValues {Curve {keys [
          {arriveTangent=4.807, interpMode=RCIM_Cubic, leaveTangent=4.807, tangentMode=RCTM_User, time=0, value=0.8}
          {arriveTangent=0.8929, interpMode=RCIM_Cubic, leaveTangent=0.8929, tangentMode=RCTM_User, time=1, value=2}
         ]}}}
      Scale Curve=0.3
     )))
   &DynamicMaterialParameters(
    Index 0 Param 1=&RandomRangeFloat(Minimum=0, Maximum=1)
    Index 0 Param 2=&FloatFromCurve(
     FloatCurve=DataInterface() {propertyValues {Curve {keys [{interpMode=RCIM_Cubic, tangentMode=RCTM_User, time=0.33, value=0}, {interpMode=RCIM_Cubic, leaveTangent=1.2673, time=0.8, value=2}]}}}
     Scale Curve=0
    )
    Index 0 Param 3=&FloatFromCurve(FloatCurve=DataInterface() {propertyValues {Curve {keys [{tangentMode=RCTM_User, time=0, value=1}, {tangentMode=RCTM_User, time=1, value=1}]}}})
    Index 0 Param 4=&FloatFromCurve(FloatCurve=DataInterface() {propertyValues {Curve {keys [
         {arriveTangent=-14.9399, interpMode=RCIM_Cubic, leaveTangent=-14.9399, tangentMode=RCTM_User, time=0, value=1}
         {interpMode=RCIM_Cubic, tangentMode=RCTM_User, time=0.15, value=0}
         {interpMode=RCIM_Cubic, tangentMode=RCTM_User, time=0.33, value=0}
         {arriveTangent=2.7964, interpMode=RCIM_Cubic, leaveTangent=2.7964, tangentMode=RCTM_User, time=1, value=1}
        ]}}})
    Index 1 Param 2=&FloatFromCurve(FloatCurve=DataInterface() {propertyValues {Curve {keys [
         {arriveTangent=23.2225, interpMode=RCIM_Cubic, leaveTangent=23.2225, tangentMode=RCTM_User, time=0, value=0}
         <at 0.1 1 RCIM_Cubic>
         {arriveTangent=-3.2993, interpMode=RCIM_Cubic, leaveTangent=-3.2993, tangentMode=RCTM_User, time=1, value=0}
        ]}}})
    Module.Param0WriteEnabled=true
    Module.Param0WriteEnabled1=true
    Module.Param1WriteEnabled=true
    Module.Param1WriteEnabled1=true
    Module.Param2WriteEnabled=true
    Module.Param2WriteEnabled1=false
    Module.Param3WriteEnabled=true
    Module.Param3WriteEnabled1=false
    Write Parameter Index 1=true
   ) {module=DynamicMaterialParameters003}
  ]
  rendererClasses [&NiagaraSpriteRendererProperties]
  renderers [{rendererClass=&NiagaraSpriteRendererProperties}]
 }
 Water {
  emitterUpdateScript [
   &EmitterState(Life Cycle Mode=::NewEnumerator1, Loop Behavior=::NewEnumerator1, Loop Duration=1, UseLoopDelay=true)
   &SpawnBurst_Instantaneous(Spawn Count=20, Use Spawn Probability=false, UseLoopCountLimit=false)
  ]
  particleSpawnScript [
   &InitializeParticle(
    Color=<rgba 0.7917 1.46 1.5833 1>
    Lifetime Max=1
    Lifetime Min=0.5
    Lifetime Mode=::NewEnumerator1
    Position Offset Coordinate Space=::Local
    Ribbon Width=3
    Ribbon Width Mode=::NewEnumerator1
    Sprite Rotation Mode=::NewEnumerator2
    Sprite Size Mode=::NewEnumerator1
    Uniform Sprite Size Max=20
    Uniform Sprite Size Min=15
    UsePositionOffset=true
    Write Material Random=true
   )
   &ShapeLocation(
    Handle Radius=15
    Large Radius=20
    MaskSpawnGroup=false
    Offset Mode=::NewEnumerator0
    Override Seed=false
    Rotation Angle=1
    Rotation Mode=::NewEnumerator1
    Shape Primitive=::NewEnumerator3
    UseOwnerScale=true
   )
   &SkeletalMeshLocation(
    Bone / Socket Index=1
    MaskSpawnGroup=false
    Mesh Sampling Type=::NewEnumerator5
    Orientation Sampling=::NewEnumerator1
    Sampled Position Offset=&AddVectorToPosition(Position=<-40 70 170>, Vector=$Particles.ShapeLocation.ShapeVector)
    Skeletal Mesh=$System.Skeletal Mesh
    Socket Sampling Mode=::NewEnumerator5
   )
   &AddVelocity(
    Module.Use Falloff Distance=false
    Origin Offset Coordinate Space=::Simulation
    Override Seed=false
    Velocity Mode=::NewEnumerator1
    Velocity Speed=&RandomRangeFloat(Minimum=100, Maximum=150)
   )
   /Niagara/Modules/Update/SubUV/V2/SubUVAnimation(End Frame=3, SubUV Animation Mode:ENiagara_SubUVLookupMode=::NewEnumerator2)
  ]
  particleUpdateScript [
   &ParticleState()
   &Drag(Drag=2, Use Linear Drag=true, Use Rotational Drag=false)
   &GravityForce(Gravity=<0 0 -300>)
   &InheritSourceMovement(ApplyForce=false, ApplyPosition=true, ApplyPositionFalloff=false)
   &SolveForcesAndVelocity(Clamp Velocity=false, Limit Acceleration=false)
   &ScaleColor(
    Scale Alpha=&FloatFromCurve(
     FloatCurve=DataInterface() {propertyValues {Curve {keys [<at 0 -1>, {interpMode=RCIM_Cubic, tangentMode=RCTM_User, time=0.1, value=0.9316}, <at 0.7 0.875>, <at 1 -1>]}}}
     Scale Curve=0.3
    )
    Scale RGB=&VectorFromCurve(VectorCurve=DataInterface() {propertyValues {XCurve {keys [<at 0 2>,<at 1 0>]}, YCurve {keys [<at 0 2>,<at 1 0>]}, ZCurve {keys [<at 0 2>,<at 1 0>]}}})
    ScaleA=true
    ScaleRGB=true
   ) {module=ScaleColor001}
   &ScaleSpriteSize(Scale Factor=&Vector2DFromFloat(Value=&FloatFromCurve(
      FloatCurve=DataInterface() {propertyValues {Curve {keys [
          {arriveTangent=4.807, interpMode=RCIM_Cubic, leaveTangent=4.807, tangentMode=RCTM_User, time=0, value=0.7}
          {arriveTangent=0.8929, interpMode=RCIM_Cubic, leaveTangent=0.8929, tangentMode=RCTM_User, time=1, value=4}
         ]}}}
      Scale Curve=&FN_FrontEndFloatModifier(Gameplay=0.6, Front End=0.4)
     )))
   &DynamicMaterialParameters(
    Index 0 Param 1=&FloatFromCurve(FloatCurve=DataInterface() {propertyValues {Curve {keys [
         {arriveTangent=-2.1712, interpMode=RCIM_Cubic, leaveTangent=-2.1712, tangentMode=RCTM_User, time=0.25, value=1}
         {arriveTangent=-0.3234, interpMode=RCIM_Cubic, leaveTangent=-0.3234, tangentMode=RCTM_User, time=1.0011, value=0}
        ]}}})
    Index 0 Param 2=&FloatFromCurve(FloatCurve=DataInterface() {propertyValues {Curve {keys [<at -0.0086 -1.3001>,<at 0.9945 1.9153>]}}})
    Index 0 Param 3=0.5
    Module.Param0WriteEnabled=true
    Module.Param1WriteEnabled=false
    Module.Param2WriteEnabled=true
    Module.Param3WriteEnabled=true
   ) {module=DynamicMaterialParameters001}
  ]
  rendererClasses [&NiagaraSpriteRendererProperties]
  renderers [{rendererClass=&NiagaraSpriteRendererProperties}]
 }
}
summary {
 systemName=NS_Emote_Phantom_Umbrella_Water
 emitters [
  {emitterName=Water, rendererClasses [&NiagaraSpriteRendererProperties]}
  {emitterName=Drop, rendererClasses [&NiagaraSpriteRendererProperties]}
 ]
 userVariables [
  User.bIsFrontEnd {type=NiagaraBool}
  User.bIsFrontEndPreview {type=NiagaraBool}
 ]
}
